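# Movie Render Queue burn-in widget that overlays job, build and asset information on asset showcase sequence renders.
# SequencerTick drives per-region update functions for sequence, world, author, date, time of day and asset name.
# Changelist and asset name come from the SHOWCASE_Globals save, and branch is parsed from the engine version.
# CurrentLOD supplies LOD vertex text; camera and shot fields are unwired, while master and sample fields show NA.
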WidgetBlueprint /Game/NeverCook/ArtTools/AssetShowcase/MRQAssets/BP_Showcase_BurnIn_Seq
graphs [
 EventGraph {nodes [
   n1 = SequencerTick!() -> (OutputDelegate, then, sequence:Movie Scene Sequence Object Reference, TOD:Fort Day Sequence Actor Object Reference) {is=title K2Node_CustomEvent}
   > Is Valid(InputObject=@n1.sequence)
   > n2 = Sequence()
   > Get Changelist() {is=title K2Node_CallFunction}
   n3 = Sequence!(execute=@n2.then_1) -> (then_0, then_1, then_2:Exec, then_3:Exec, then_4:Exec, then_5:Exec, then_6:Exec, then_7:Exec)
   > UpdateTopLeftText!(execute, self, seq:Movie Scene Sequence Object Reference=@n1.sequence) -> (then)
   UpdateTopRightText!(execute=@n3.then_1, self, Seq:Movie Scene Sequence Object Reference=@n1.sequence) -> (then)
   UpdateBottomLeftText!(execute=@n3.then_2, self, Seq:Movie Scene Sequence Object Reference=@n1.sequence) -> (then)
   UpdateBottomCenter!(execute=@n3.then_3, self, Seq:Movie Scene Sequence Object Reference=@n1.sequence, TOD:Fort Day Sequence Actor Object Reference=@n1.TOD) -> (then)
   UpdateBottomRightText!(execute=@n3.then_4, self, Seq:Movie Scene Sequence Object Reference=@n1.sequence) -> (then)
   UpdateUserCommentText!(execute=@n3.then_5, self, Seq:Movie Scene Sequence Object Reference=@n1.sequence) -> (then)
   UpdateAssetName(execute=@n3.then_6)
   n4 = Event CurrentLOD!() -> (OutputDelegate, then, Vertices:int=0, CurLOD:String) {is=title K2Node_Event}
   > n5 = Set LODVertices(LODVertices:int=@n4.Vertices) -> (Output_Get:int=0)
   > Set LODInfoString(LODInfoString:String=Append!(A=@n4.CurLOD, B=": ", C:String="To String (Integer)"(InInt=@n5), D:String=" vtx") -> (ReturnValue)) -> (Output_Get:String)
  ]}
 Get Changelist {nodes [
   Get Changelist() {is=title K2Node_FunctionEntry}
   > Do Once()
   > n1 = DoesSaveGameExist!(execute, SlotName:String=SHOWCASE_Globals, UserIndex:int=0) -> (then, ReturnValue:bool)
   > n2 = Branch(Condition=@n1)
   > n3 = LoadGameFromSlot!(execute, SlotName:String=SHOWCASE_Globals, UserIndex:int=0) -> (then, ReturnValue:Save Game Object Reference)
   > n4 = Cast To SG_ShowcaseRel!(execute, Object=@n3) -> (then, CastFailed, AsSG Showcase Rel:SG Showcase Rel Object Reference) {is=title K2Node_DynamicCast}
   > n5 = Set SGShowcase(SGShowcase:SG Showcase Rel Object Reference=@n4) -> (Output_Get:SG Showcase Rel Object Reference)
   > Set Change(Change:int=Get changelist!(self:SG Showcase Rel Object Reference=@n5) -> (changelist:int=0)) -> (Output_Get:int=0)
   Set Change(execute=@n2.else, Change=-1) -> (Output_Get=0)
  ]}
 ResolveBinding {nodes [
   ResolveBinding!() -> (then, binding:Movie Scene Binding Proxy Structure) {is=title K2Node_FunctionEntry}
   > Return Node!(execute, instance:Actor Object Reference)
  ]}
 SetSequencerTOD {nodes [
   SetSequencerTOD!() -> (then, TODActor:Fort Day Sequence Actor Object Reference, fStart:int=0, fStop:int=0, minVal:double, maxVal:double) {is=title K2Node_FunctionEntry}
   > Return Node!(execute, TODValid:bool=false)
  ]}
 UpdateAssetName {nodes [
   UpdateAssetName() {is=title K2Node_FunctionEntry}
   n1 = Get SGShowcase() -> (SGShowcase:SG Showcase Rel Object Reference)
   > "SetText (Text)"(
    self=Get AssetNameText() -> (AssetNameText:Common Text Object Reference)
    InText="To Text (String)"(InString=Get assetName!(self:SG Showcase Rel Object Reference=@n1) -> (assetName:String))
   )
  ]}
 UpdateBottomCenter {nodes [
   n1 = UpdateBottomCenter!() -> (then, Seq:Movie Scene Sequence Object Reference, TOD:Fort Day Sequence Actor Object Reference) {is=title K2Node_FunctionEntry}
   > n2 = Sequence!(execute) -> (then_0, then_1, then_2)
   > "SetText (Text)"(
    self=Get BottomCenter() -> (BottomCenter:Text Object Reference)
    InText=Format Text!(Format="LOCTEXT(\"{FocalLength}mm   f/{FStop}   {FocusDistance}\")", FocalLength:Wildcard, FStop:Wildcard, FocusDistance:Wildcard) -> (Result)
   )
   Is Valid(exec=@n2.then_2, InputObject=@n1.TOD)
   n3 = GetStaticTimeOfDay!(self:Day Sequence Actor Object Reference=@n1.TOD) -> (ReturnValue:float)
   n4 = "To String (Integer)"(InInt=Floor(A=@n3) {is=title K2Node_CallFunction})
   n5 = Fraction!(A:double=@n3) -> (ReturnValue:double)
   n6 = "To String (Integer)"(InInt=Truncate(A=float * float(A=@n5, B=60) {is=title K2Node_PromotableOperator}))
   > "SetText (Text)"(
    self=Get BottomLeft_tod() -> (BottomLeft_tod:Text Object Reference)
    InText=Format Text!(Format="LOCTEXT(\"Time Of Day: {curTOD}\")", curTOD:String=Append!(A=@n4, B=:, C=@n6) -> (ReturnValue)) -> (Result)
   )
   "SetText (Text)"(
    execute=@n2.then_1
    self=Get BottomLeft_FOV() -> (BottomLeft_FOV:Text Object Reference)
    InText=Format Text!(Format="LOCTEXT(\"FOV: {curFOV}\")", curFOV:Wildcard) -> (Result)
   )
   GetTimeOfDay!() -> (ReturnValue:float)
   Truncate(A=getFOV!(focalLength:double=GetCurrentFocalLength!(InMoviePipeline:Movie Pipeline Object Reference) -> (ReturnValue:float)) -> (ReturnValue:double))
   GetCurrentFocusDistance!(InMoviePipeline:Movie Pipeline Object Reference) -> (ReturnValue:float)
   GetCurrentAperture!(InMoviePipeline:Movie Pipeline Object Reference) -> (ReturnValue:float)
  ]}
 UpdateBottomLeftText {nodes [
   UpdateBottomLeftText!() -> (then, Seq:Movie Scene Sequence Object Reference) {is=title K2Node_FunctionEntry}
   > "SetText (Text)"(self=Get BottomLeft() -> (BottomLeft:Text Object Reference), InText=Format Text!(Format="LOCTEXT(\"Master:  NA \")") -> (Result))
  ]}
 UpdateBottomRightText {nodes [
   UpdateBottomRightText!() -> (then, Seq:Movie Scene Sequence Object Reference) {is=title K2Node_FunctionEntry}
   > "SetText (Text)"(
    self=Get BottomRight() -> (BottomRight:Text Object Reference)
    InText=Format Text!(Format="LOCTEXT(\"Shot:   {ShotTimecode} ({ShotFrameNumber}f)\")", ShotTimecode:Wildcard, ShotFrameNumber:Wildcard) -> (Result)
   )
   > "SetText (Text)"(
    self=Get BottomRight2() -> (BottomRight2:Text Object Reference)
    InText=Format Text!(Format="LOCTEXT(\"{ShotName}\")", ShotName:Wildcard) -> (Result)
   )
   > "SetText (Multi-Line Text Box)"!(
    execute
    self:"Text Box (Multi- Line) Object Reference"=Get LODInfo() -> (LODInfo:"Text Box (Multi- Line) Object Reference")
    InText:Text="To Text (String)"(InString=Get LODInfoString() -> (LODInfoString:String))
   ) -> (then)
   Break Frame Number!(FrameNumber:"Frame Number Structure (by ref)"=GetCurrentShotFrameNumber!(InMoviePipeline:Movie Pipeline Object Reference) -> (ReturnValue:Frame Number Structure)) -> (Value:int) {is=title K2Node_BreakStruct}
   GetCurrentSegmentName!(InMoviePipeline:Movie Pipeline Object Reference) -> (OutOuterName:Text, OutInnerName:Text)
   "To Text (String)"(InString=Timecode To String!(
     InTimecode:"Timecode Structure (by ref)"=GetCurrentShotTimecode!(InMoviePipeline:Movie Pipeline Object Reference) -> (ReturnValue:Timecode Structure)
     bForceSignDisplay:bool=false
    ) -> (ReturnValue:String))
  ]}
 UpdateTopLeftText {nodes [
   n1 = UpdateTopLeftText!() -> (then, seq:Movie Scene Sequence Object Reference) {is=title K2Node_FunctionEntry}
   > n2 = GetCurrentLevelName!(execute, bRemovePrefixString:bool=true) -> (then, ReturnValue:String)
   n3 = GetDisplayName(Object=@n1)
   > "SetText (Text)"(
    self=Get Top_01() -> (Top_01:Text Object Reference)
    InText=Format Text!(Format="LOCTEXT(\"Job: {SequenceName}\")", SequenceName:Text="To Text (String)"(InString=@n3)) -> (Result)
   )
   n4 = UTC Now!() -> (ReturnValue:Date Time Structure)
   > "SetText (Text)"(
    self=Get Top_02() -> (Top_02:Text Object Reference)
    InText=Format Text!(
     Format="LOCTEXT(\"{Time} {Date}\")"
     Time:Text="As Time (from UTC)"!(InDateTime:"Date Time Structure (by ref)"=@n4, InTimeZone:String, InTimeStyle:EDateTimeStyle Enum=Default) -> (ReturnValue:Text)
     Date:Text="As Date (from UTC)"!(InDateTime:"Date Time Structure (by ref)"=@n4, InTimeZone:String, InDateStyle:EDateTimeStyle Enum=Default) -> (ReturnValue:Text)
    ) -> (Result)
   )
   > "SetText (Text)"(
    self=Get Top_03() -> (Top_03:Text Object Reference)
    InText=Format Text!(Format="LOCTEXT(\"World: {WorldName}\")", WorldName:Text="To Text (String)"(InString=@n2)) -> (Result)
   )
   n5 = GetPlatformUserName!() -> (ReturnValue:String)
   > "SetText (Text)"(
    self=Get Top_04() -> (Top_04:Text Object Reference)
    InText=Format Text!(Format="LOCTEXT(\"Author: {AuthorName}\")", AuthorName:Text="To Text (String)"(InString=@n5)) -> (Result)
   )
  ]}
 UpdateTopRightText {nodes [
   UpdateTopRightText!() -> (then, Seq:Movie Scene Sequence Object Reference) {is=title K2Node_FunctionEntry}
   > n1 = Sequence!(execute) -> (then_0, then_1, then_2, then_3:Exec, then_4:Exec)
   n2 = Get Change() -> (Change:int=0)
   > n3 = Branch(Condition="Equal (Integer)"(A=@n2, B=-1))
   > Set ChangeList(ChangeList:String=CL - Unknown) -> (Output_Get:String)
   Set ChangeList(execute=@n3.else, ChangeList=Append!(A="CL - ", B="To String (Integer)"(InInt=@n2)) -> (ReturnValue))
   n4 = Get BranchName() -> (BranchName:String)
   n5 = Get ChangeList() -> (ChangeList:String)
   "SetText (Text)"(
    execute=@n1.then_2
    self=Get TopRight() -> (TopRight:Text Object Reference)
    InText=Format Text!(
     Format="LOCTEXT(\"{BranchName} - {ChangeList}\")"
     BranchName:Text="To Text (String)"(InString=@n4)
     ChangeList:Text="To Text (String)"(InString=@n5)
    ) -> (Result)
   )
   n6 = GetEngineVersion!() -> (ReturnValue:String)
   n7 = FindSubstring(SearchIn=@n6, Substring=+, bSearchFromEnd=true)
   Branch(execute=@n1.then_1, Condition="integer >= integer"!(A:int=@n7, B:int=0) -> (ReturnValue:bool) {is=title K2Node_CallFunction})
   > Set BranchName(BranchName:String=GetSubstring!(
     SourceString:String=@n6
     StartIndex:int=int + int(A=@n7, B=1) -> (ReturnValue=0) {is=title K2Node_CommutativeAssociativeBinaryOperator}
     Length:int=int - int(A=@n7, B=1) {is=title K2Node_CallFunction}
    ) -> (ReturnValue:String)) -> (Output_Get:String)
   "SetText (Text)"(
    execute=@n1.then_3
    self=Get BottomLeft2() -> (BottomLeft2:Text Object Reference)
    InText=Format Text!(Format="LOCTEXT(\"TSS: {SampleCount}\")", SampleCount:String=MakeLiteralString(Value=NA)) -> (Result)
   )
   n8 = GetTimeOfDay()
   n9 = "To String (Integer)"(InInt=Floor(A=@n8))
   n10 = Fraction(A=@n8)
   n11 = "To String (Integer)"(InInt=Truncate(A=float * float(A=@n10, B=60)))
   "SetText (Text)"(
    self=Get BottomLeft_tod()
    InText=Format Text!(Format="LOCTEXT(\"Time Of Day: {curTOD}\")", curTOD:String=Append!(A=@n9, B=:, C:String=@n11) -> (ReturnValue)) -> (Result)
   )
  ]}
 UpdateUserCommentText {nodes [
   UpdateUserCommentText!() -> (then, Seq:Movie Scene Sequence Object Reference) {is=title K2Node_FunctionEntry}
   > Set Text(self:"Text Box (Multi- Line) Object Reference"=Get UserJobComment() -> (UserJobComment:"Text Box (Multi- Line) Object Reference"))
  ]}
 getFOV {nodes [
   n1 = getFOV!() -> (then, focalLength:double=~) {is=title K2Node_FunctionEntry}
   n2 = float / float(A=11.88, B=@n1)
   > Return Node!(execute, ReturnValue:double=float * float(A="Atan (Degrees)"!(A:double=@n2) -> (ReturnValue:double), B=2))
  ]}
]
variables [
 Change {type=integer, default=-1}
 LODInfoString {type=string, default=""}
 LODVertices {type=integer}
 SGShowcase {type=SG_ShowcaseRel_C, default=null}
]
widgets {
 info {
  inheritedWidgetCount=0
  parentClass=/Script/MovieRenderPipelineSettings.MoviePipelineBurnInWidget
  rootWidgetClass=&Overlay
  widgetCount=35
 }
 widgets [
  Overlay_45 {widgetClassPath=&Overlay, widget=Overlay_45}
  RenderedImagePlaceholder {
   widgetClassPath=&Image
   parent=Overlay_45
   widget=RenderedImagePlaceholder
   properties {
    visibility=Hidden
    brush {
     bIsDynamicallyLoaded=false
     drawAs=Image
     imageType=NoImage
     mirroring=NoMirror
     resourceName=None
     resourceObject=null
     tiling=NoTile
     imageSize=<1920 1080>
     margin=<ltrb 0 0 0 0>
     outlineSettings {
      bUseBrushTransparency=false
      roundingType=HalfHeightRadius
      width=0
      color {colorUseRule=UseColor_Specified, specifiedColor=<rgba 0 0 0 0>}
      cornerRadii=<0 0 0 0>
     }
     tintColor {colorUseRule=UseColor_Specified, specifiedColor=<rgba 0.026 0.026 0.026 1>}
     uVRegion {bIsValid=false, max=<0 0>, min=<0 0>}
    }
   }
   slot {class=OverlaySlot, properties {horizontalAlignment=HAlign_Fill, verticalAlignment=VAlign_Fill}}
  }
  VerticalBox_41 {
   widgetClassPath=&VerticalBox
   parent=Overlay_45
   widget=VerticalBox_41
   slot {class=OverlaySlot, properties {horizontalAlignment=HAlign_Fill, verticalAlignment=VAlign_Fill}}
  }
  Border_23 {
   widgetClassPath=&Border
   parent=VerticalBox_41
   widget=Border_23
   properties {
    background {
     bIsDynamicallyLoaded=false
     drawAs=Image
     imageType=NoImage
     mirroring=NoMirror
     resourceName=None
     resourceObject=null
     tiling=NoTile
     imageSize=<32 32>
     margin=<ltrb 0 0 0 0>
     outlineSettings {
      bUseBrushTransparency=false
      roundingType=HalfHeightRadius
      width=0
      color {colorUseRule=UseColor_Specified, specifiedColor=<rgba 0 0 0 0>}
      cornerRadii=<0 0 0 0>
     }
     tintColor {colorUseRule=UseColor_Specified, specifiedColor=<rgba 0 0 0 0.5>}
     uVRegion {bIsValid=false, max=<0 0>, min=<0 0>}
    }
    padding=<ltrb 4 48 4 4>
   }
   slot {class=VerticalBoxSlot, properties {verticalAlignment=VAlign_Top}}
  }
  HorizontalBox_22 {
   widgetClassPath=&HorizontalBox
   parent=Border_23
   widget=HorizontalBox_22
   slot {class=BorderSlot, properties {padding=<ltrb 4 48 4 4>}}
  }
  Top_01 {
   widgetClassPath=&TextBlock
   bIsVariable=true
   parent=HorizontalBox_22
   widget=Top_01
   properties {
    text=Job: PipelineTestSeq
    font {
     bForceMonospaced=false
     bMaterialIsStencil=false
     fontMaterial=null
     fontObject=&NotoSans
     letterSpacing=0
     monospacedWidth=1
     size=18
     skewAmount=0
     typefaceFontName=NotoSans-Regular
     outlineSettings {
      bApplyOutlineToDropShadows=false
      bMiteredCorners=false
      bSeparateFillAlpha=false
      outlineMaterial=null
      outlineSize=0
      outlineColor=<rgba 0 0 0 1>
     }
    }
   }
   slot {class=HorizontalBoxSlot, properties {horizontalAlignment=HAlign_Left}}
  }
  Spacer_125 {
   widgetClassPath=&Spacer
   parent=HorizontalBox_22
   widget=Spacer_125
   slot {class=HorizontalBoxSlot, properties {size {sizeRule=Fill, value=1}}}
  }
  Top_02 {
   widgetClassPath=&TextBlock
   bIsVariable=true
   parent=HorizontalBox_22
   widget=Top_02
   properties {
    text=03:44pm 01/19/2020
    font {
     bForceMonospaced=false
     bMaterialIsStencil=false
     fontMaterial=null
     fontObject=&NotoSans
     letterSpacing=0
     monospacedWidth=1
     size=18
     skewAmount=0
     typefaceFontName=NotoSans-Regular
     outlineSettings {
      bApplyOutlineToDropShadows=false
      bMiteredCorners=false
      bSeparateFillAlpha=false
      outlineMaterial=null
      outlineSize=0
      outlineColor=<rgba 0 0 0 1>
     }
    }
   }
   slot {class=HorizontalBoxSlot, properties {horizontalAlignment=HAlign_Center}}
  }
  Spacer {
   widgetClassPath=&Spacer
   parent=HorizontalBox_22
   widget=Spacer
   slot {class=HorizontalBoxSlot, properties {size {sizeRule=Fill, value=1}}}
  }
  Top_03 {
   widgetClassPath=&TextBlock
   bIsVariable=true
   parent=HorizontalBox_22
   widget=Top_03
   properties {
    text=World: MoviePipelineTestMap
    font {
     bForceMonospaced=false
     bMaterialIsStencil=false
     fontMaterial=null
     fontObject=&NotoSans
     letterSpacing=0
     monospacedWidth=1
     size=18
     skewAmount=0
     typefaceFontName=NotoSans-Regular
     outlineSettings {
      bApplyOutlineToDropShadows=false
      bMiteredCorners=false
      bSeparateFillAlpha=false
      outlineMaterial=null
      outlineSize=0
      outlineColor=<rgba 0 0 0 1>
     }
    }
   }
   slot {class=HorizontalBoxSlot, properties {horizontalAlignment=HAlign_Center}}
  }
  Spacer_1 {
   widgetClassPath=&Spacer
   parent=HorizontalBox_22
   widget=Spacer_1
   slot {class=HorizontalBoxSlot, properties {size {sizeRule=Fill, value=1}}}
  }
  Top_04 {
   widgetClassPath=&TextBlock
   bIsVariable=true
   parent=HorizontalBox_22
   widget=Top_04
   properties {
    text=Author: MannyMannequin
    font {
     bForceMonospaced=false
     bMaterialIsStencil=false
     fontMaterial=null
     fontObject=&NotoSans
     letterSpacing=0
     monospacedWidth=1
     size=18
     skewAmount=0
     typefaceFontName=NotoSans-Regular
     outlineSettings {
      bApplyOutlineToDropShadows=false
      bMiteredCorners=false
      bSeparateFillAlpha=false
      outlineMaterial=null
      outlineSize=0
      outlineColor=<rgba 0 0 0 1>
     }
    }
   }
   slot {class=HorizontalBoxSlot, properties {horizontalAlignment=HAlign_Center, verticalAlignment=VAlign_Center}}
  }
  Spacer_2 {
   widgetClassPath=&Spacer
   parent=HorizontalBox_22
   widget=Spacer_2
   slot {class=HorizontalBoxSlot, properties {size {sizeRule=Fill, value=1}}}
  }
  TopRight {
   widgetClassPath=&TextBlock
   bIsVariable=true
   parent=HorizontalBox_22
   widget=TopRight
   properties {
    justification=Right
    text=Dev-Editor - ugs:11047590
    font {
     bForceMonospaced=false
     bMaterialIsStencil=false
     fontMaterial=null
     fontObject=&NotoSans
     letterSpacing=0
     monospacedWidth=1
     size=18
     skewAmount=0
     typefaceFontName=NotoSans-Regular
     outlineSettings {
      bApplyOutlineToDropShadows=false
      bMiteredCorners=false
      bSeparateFillAlpha=false
      outlineMaterial=null
      outlineSize=0
      outlineColor=<rgba 0 0 0 1>
     }
    }
   }
   slot {class=HorizontalBoxSlot}
  }
  Spacer_60 {
   widgetClassPath=&Spacer
   parent=VerticalBox_41
   widget=Spacer_60
   slot {class=VerticalBoxSlot, properties {size {sizeRule=Fill, value=1}}}
  }
  Border_70 {
   widgetClassPath=&Border
   parent=VerticalBox_41
   widget=Border_70
   properties {
    background {
     bIsDynamicallyLoaded=false
     drawAs=Image
     imageType=NoImage
     mirroring=NoMirror
     resourceName=None
     resourceObject=null
     tiling=NoTile
     imageSize=<32 32>
     margin=<ltrb 0 0 0 0>
     outlineSettings {
      bUseBrushTransparency=false
      roundingType=HalfHeightRadius
      width=0
      color {colorUseRule=UseColor_Specified, specifiedColor=<rgba 0 0 0 0>}
      cornerRadii=<0 0 0 0>
     }
     tintColor {colorUseRule=UseColor_Specified, specifiedColor=<rgba 1 1 1 0>}
     uVRegion {bIsValid=false, max=<0 0>, min=<0 0>}
    }
    padding=<ltrb 0 0 0 0>
   }
   slot {class=VerticalBoxSlot, properties {verticalAlignment=VAlign_Bottom, size {sizeRule=Fill, value=1}}}
  }
  HorizontalBox_231 {
   widgetClassPath=&HorizontalBox
   parent=Border_70
   widget=HorizontalBox_231
   slot {class=BorderSlot, properties {padding=<ltrb 0 0 0 0>}}
  }
  UserJobComment {
   widgetClassPath=/Script/UMG.MultiLineEditableTextBox
   bIsVariable=true
   parent=HorizontalBox_231
   widget=UserJobComment
   properties {
    text="CL## 25029801"
    widgetStyle {
     backgroundColor {colorUseRule=UseColor_Specified, specifiedColor=<rgba 1 1 1 1>}
     backgroundImageFocused {
      bIsDynamicallyLoaded=false
      drawAs=Box
      imageType=FullColor
      mirroring=NoMirror
      resourceName=../../../Engine/Content/Slate/Common/TextBox_Hovered.png
      resourceObject=null
      tiling=NoTile
      imageSize=<0 0>
      margin=<ltrb 0.25 0.25 0.25 0.25>
      outlineSettings {
       bUseBrushTransparency=false
       roundingType=HalfHeightRadius
       width=0
       color {colorUseRule=UseColor_Specified, specifiedColor=<rgba 0 0 0 0>}
       cornerRadii=<0 0 0 0>
      }
      tintColor {colorUseRule=UseColor_Specified, specifiedColor=<rgba 1 1 1 0>}
      uVRegion {bIsValid=false, max=<0 0>, min=<0 0>}
     }
     backgroundImageHovered {
      bIsDynamicallyLoaded=false
      drawAs=Box
      imageType=FullColor
      mirroring=NoMirror
      resourceName=../../../Engine/Content/Slate/Common/TextBox_Hovered.png
      resourceObject=null
      tiling=NoTile
      imageSize=<0 0>
      margin=<ltrb 0.25 0.25 0.25 0.25>
      outlineSettings {
       bUseBrushTransparency=false
       roundingType=HalfHeightRadius
       width=0
       color {colorUseRule=UseColor_Specified, specifiedColor=<rgba 0 0 0 0>}
       cornerRadii=<0 0 0 0>
      }
      tintColor {colorUseRule=UseColor_Specified, specifiedColor=<rgba 1 1 1 0>}
      uVRegion {bIsValid=false, max=<0 0>, min=<0 0>}
     }
     backgroundImageNormal {
      bIsDynamicallyLoaded=false
      drawAs=Box
      imageType=FullColor
      mirroring=NoMirror
      resourceName=../../../Engine/Content/Slate/Common/TextBox.png
      resourceObject=null
      tiling=NoTile
      imageSize=<0 0>
      margin=<ltrb 0.25 0.25 0.25 0.25>
      outlineSettings {
       bUseBrushTransparency=false
       roundingType=HalfHeightRadius
       width=0
       color {colorUseRule=UseColor_Specified, specifiedColor=<rgba 0 0 0 0>}
       cornerRadii=<0 0 0 0>
      }
      tintColor {colorUseRule=UseColor_Specified, specifiedColor=<rgba 1 1 1 0>}
      uVRegion {bIsValid=false, max=<0 0>, min=<0 0>}
     }
     backgroundImageReadOnly {
      bIsDynamicallyLoaded=false
      drawAs=Box
      imageType=FullColor
      mirroring=NoMirror
      resourceName=../../../Engine/Content/Slate/Common/TextBox_ReadOnly.png
      resourceObject=null
      tiling=NoTile
      imageSize=<0 0>
      margin=<ltrb 0.25 0.25 0.25 0.25>
      outlineSettings {
       bUseBrushTransparency=false
       roundingType=HalfHeightRadius
       width=0
       color {colorUseRule=UseColor_Specified, specifiedColor=<rgba 0 0 0 0>}
       cornerRadii=<0 0 0 0>
      }
      tintColor {colorUseRule=UseColor_Specified, specifiedColor=<rgba 1 1 1 1>}
      uVRegion {bIsValid=false, max=<0 0>, min=<0 0>}
     }
     focusedForegroundColor {colorUseRule=UseColor_Foreground, specifiedColor=<rgba 1 0 1 1>}
     foregroundColor {colorUseRule=UseColor_Foreground, specifiedColor=<rgba 1 0 1 1>}
     hScrollBarPadding=<ltrb 0 0 0 0>
     padding=<ltrb 4 2 4 2>
     readOnlyForegroundColor {colorUseRule=UseColor_Foreground, specifiedColor=<rgba 1 0 1 1>}
     scrollBarStyle {
      thickness=16
      draggedThumbImage {
       bIsDynamicallyLoaded=false
       drawAs=Box
       imageType=FullColor
       mirroring=NoMirror
       resourceName=../../../Engine/Content/Slate/Common/Scrollbar_Thumb.png
       resourceObject=null
       tiling=NoTile
       imageSize=<0 0>
       margin=<ltrb 0.25 0.25 0.25 0.25>
       outlineSettings {
        bUseBrushTransparency=false
        roundingType=HalfHeightRadius
        width=0
        color {colorUseRule=UseColor_Specified, specifiedColor=<rgba 0 0 0 0>}
        cornerRadii=<0 0 0 0>
       }
       tintColor {colorUseRule=UseColor_Specified, specifiedColor=<rgba 1 1 1 1>}
       uVRegion {bIsValid=false, max=<0 0>, min=<0 0>}
      }
      horizontalBackgroundImage {
       bIsDynamicallyLoaded=false
       drawAs=NoDrawType
       imageType=NoImage
       mirroring=NoMirror
       resourceName=None
       resourceObject=null
       tiling=NoTile
       imageSize=<0 0>
       margin=<ltrb 0 0 0 0>
       outlineSettings {
        bUseBrushTransparency=false
        roundingType=HalfHeightRadius
        width=0
        color {colorUseRule=UseColor_Specified, specifiedColor=<rgba 0 0 0 0>}
        cornerRadii=<0 0 0 0>
       }
       tintColor {colorUseRule=UseColor_Specified, specifiedColor=<rgba 1 1 1 1>}
       uVRegion {bIsValid=false, max=<0 0>, min=<0 0>}
      }
      horizontalBottomSlotImage {
       bIsDynamicallyLoaded=false
       drawAs=Image
       imageType=FullColor
       mirroring=NoMirror
       resourceName=../../../Engine/Content/Slate/Common/Scrollbar_Background_Horizontal.png
       resourceObject=null
       tiling=NoTile
       imageSize=<8 8>
       margin=<ltrb 0 0 0 0>
       outlineSettings {
        bUseBrushTransparency=false
        roundingType=HalfHeightRadius
        width=0
        color {colorUseRule=UseColor_Specified, specifiedColor=<rgba 0 0 0 0>}
        cornerRadii=<0 0 0 0>
       }
       tintColor {colorUseRule=UseColor_Specified, specifiedColor=<rgba 1 1 1 1>}
       uVRegion {bIsValid=false, max=<0 0>, min=<0 0>}
      }
      horizontalTopSlotImage {
       bIsDynamicallyLoaded=false
       drawAs=Image
       imageType=FullColor
       mirroring=NoMirror
       resourceName=../../../Engine/Content/Slate/Common/Scrollbar_Background_Horizontal.png
       resourceObject=null
       tiling=NoTile
       imageSize=<8 8>
       margin=<ltrb 0 0 0 0>
       outlineSettings {
        bUseBrushTransparency=false
        roundingType=HalfHeightRadius
        width=0
        color {colorUseRule=UseColor_Specified, specifiedColor=<rgba 0 0 0 0>}
        cornerRadii=<0 0 0 0>
       }
       tintColor {colorUseRule=UseColor_Specified, specifiedColor=<rgba 1 1 1 1>}
       uVRegion {bIsValid=false, max=<0 0>, min=<0 0>}
      }
      hoveredThumbImage {
       bIsDynamicallyLoaded=false
       drawAs=Box
       imageType=FullColor
       mirroring=NoMirror
       resourceName=../../../Engine/Content/Slate/Common/Scrollbar_Thumb.png
       resourceObject=null
       tiling=NoTile
       imageSize=<0 0>
       margin=<ltrb 0.25 0.25 0.25 0.25>
       outlineSettings {
        bUseBrushTransparency=false
        roundingType=HalfHeightRadius
        width=0
        color {colorUseRule=UseColor_Specified, specifiedColor=<rgba 0 0 0 0>}
        cornerRadii=<0 0 0 0>
       }
       tintColor {colorUseRule=UseColor_Specified, specifiedColor=<rgba 1 1 1 1>}
       uVRegion {bIsValid=false, max=<0 0>, min=<0 0>}
      }
      normalThumbImage {
       bIsDynamicallyLoaded=false
       drawAs=Box
       imageType=FullColor
       mirroring=NoMirror
       resourceName=../../../Engine/Content/Slate/Common/Scrollbar_Thumb.png
       resourceObject=null
       tiling=NoTile
       imageSize=<0 0>
       margin=<ltrb 0.25 0.25 0.25 0.25>
       outlineSettings {
        bUseBrushTransparency=false
        roundingType=HalfHeightRadius
        width=0
        color {colorUseRule=UseColor_Specified, specifiedColor=<rgba 0 0 0 0>}
        cornerRadii=<0 0 0 0>
       }
       tintColor {colorUseRule=UseColor_Specified, specifiedColor=<rgba 1 1 1 1>}
       uVRegion {bIsValid=false, max=<0 0>, min=<0 0>}
      }
      verticalBackgroundImage {
       bIsDynamicallyLoaded=false
       drawAs=NoDrawType
       imageType=NoImage
       mirroring=NoMirror
       resourceName=None
       resourceObject=null
       tiling=NoTile
       imageSize=<0 0>
       margin=<ltrb 0 0 0 0>
       outlineSettings {
        bUseBrushTransparency=false
        roundingType=HalfHeightRadius
        width=0
        color {colorUseRule=UseColor_Specified, specifiedColor=<rgba 0 0 0 0>}
        cornerRadii=<0 0 0 0>
       }
       tintColor {colorUseRule=UseColor_Specified, specifiedColor=<rgba 1 1 1 1>}
       uVRegion {bIsValid=false, max=<0 0>, min=<0 0>}
      }
      verticalBottomSlotImage {
       bIsDynamicallyLoaded=false
       drawAs=Image
       imageType=FullColor
       mirroring=NoMirror
       resourceName=../../../Engine/Content/Slate/Common/Scrollbar_Background_Vertical.png
       resourceObject=null
       tiling=NoTile
       imageSize=<8 8>
       margin=<ltrb 0 0 0 0>
       outlineSettings {
        bUseBrushTransparency=false
        roundingType=HalfHeightRadius
        width=0
        color {colorUseRule=UseColor_Specified, specifiedColor=<rgba 0 0 0 0>}
        cornerRadii=<0 0 0 0>
       }
       tintColor {colorUseRule=UseColor_Specified, specifiedColor=<rgba 1 1 1 1>}
       uVRegion {bIsValid=false, max=<0 0>, min=<0 0>}
      }
      verticalTopSlotImage {
       bIsDynamicallyLoaded=false
       drawAs=Image
       imageType=FullColor
       mirroring=NoMirror
       resourceName=../../../Engine/Content/Slate/Common/Scrollbar_Background_Vertical.png
       resourceObject=null
       tiling=NoTile
       imageSize=<8 8>
       margin=<ltrb 0 0 0 0>
       outlineSettings {
        bUseBrushTransparency=false
        roundingType=HalfHeightRadius
        width=0
        color {colorUseRule=UseColor_Specified, specifiedColor=<rgba 0 0 0 0>}
        cornerRadii=<0 0 0 0>
       }
       tintColor {colorUseRule=UseColor_Specified, specifiedColor=<rgba 1 1 1 1>}
       uVRegion {bIsValid=false, max=<0 0>, min=<0 0>}
      }
     }
     textStyle {
      overflowPolicy=Clip
      transformPolicy=None
      colorAndOpacity {colorUseRule=UseColor_Specified, specifiedColor=<rgba 1 1 1 1>}
      font {
       bForceMonospaced=false
       bMaterialIsStencil=false
       fontMaterial=null
       fontObject=/Engine/EngineFonts/DroidSansMono
       letterSpacing=0
       monospacedWidth=1
       size=18
       skewAmount=0
       typefaceFontName=Default
       outlineSettings {
        bApplyOutlineToDropShadows=false
        bMiteredCorners=false
        bSeparateFillAlpha=false
        outlineMaterial=null
        outlineSize=1
        outlineColor=<rgba 0 0 0 1>
       }
      }
      highlightColor {colorUseRule=UseColor_Specified, specifiedColor=<rgba 0.02 0.3 0 1>}
      highlightShape {
       bIsDynamicallyLoaded=false
       drawAs=Box
       imageType=FullColor
       mirroring=NoMirror
       resourceName=../../../Engine/Content/Slate/Common/TextBlockHighlightShape.png
       resourceObject=null
       tiling=NoTile
       imageSize=<0 0>
       margin=<ltrb 0.375 0.375 0.375 0.375>
       outlineSettings {
        bUseBrushTransparency=false
        roundingType=HalfHeightRadius
        width=0
        color {colorUseRule=UseColor_Specified, specifiedColor=<rgba 0 0 0 0>}
        cornerRadii=<0 0 0 0>
       }
       tintColor {colorUseRule=UseColor_Specified, specifiedColor=<rgba 1 1 1 1>}
       uVRegion {bIsValid=false, max=<0 0>, min=<0 0>}
      }
      selectedBackgroundColor {colorUseRule=UseColor_Foreground, specifiedColor=<rgba 1 0 1 1>}
      shadowColorAndOpacity=<rgba 0 0 0 1>
      shadowOffset=<0 0>
      strikeBrush {
       bIsDynamicallyLoaded=false
       drawAs=Image
       imageType=NoImage
       mirroring=NoMirror
       resourceName=None
       resourceObject=null
       tiling=NoTile
       imageSize=<32 32>
       margin=<ltrb 0 0 0 0>
       outlineSettings {
        bUseBrushTransparency=false
        roundingType=HalfHeightRadius
        width=0
        color {colorUseRule=UseColor_Specified, specifiedColor=<rgba 0 0 0 0>}
        cornerRadii=<0 0 0 0>
       }
       tintColor {colorUseRule=UseColor_Specified, specifiedColor=<rgba 1 1 1 1>}
       uVRegion {bIsValid=false, max=<0 0>, min=<0 0>}
      }
      underlineBrush {
       bIsDynamicallyLoaded=false
       drawAs=Image
       imageType=NoImage
       mirroring=NoMirror
       resourceName=None
       resourceObject=null
       tiling=NoTile
       imageSize=<32 32>
       margin=<ltrb 0 0 0 0>
       outlineSettings {
        bUseBrushTransparency=false
        roundingType=HalfHeightRadius
        width=0
        color {colorUseRule=UseColor_Specified, specifiedColor=<rgba 0 0 0 0>}
        cornerRadii=<0 0 0 0>
       }
       tintColor {colorUseRule=UseColor_Specified, specifiedColor=<rgba 1 1 1 1>}
       uVRegion {bIsValid=false, max=<0 0>, min=<0 0>}
      }
     }
     vScrollBarPadding=<ltrb 0 0 0 0>
    }
   }
   slot {class=HorizontalBoxSlot, properties {verticalAlignment=VAlign_Center, size {sizeRule=Fill, value=1}}}
  }
  AssetNameText {
   widgetClassPath=&CommonTextBlock
   bIsVariable=true
   parent=HorizontalBox_231
   widget=AssetNameText
   properties {
    bIsScrollingEnabled=false
    justification=Center
    style=&/AS_TextStyle_SubH4_outline.AS_TextStyle_SubH4_outline_C
    text=Asset Name
    font {
     bForceMonospaced=false
     bMaterialIsStencil=false
     fontMaterial=null
     fontObject=&BurbankSmall
     letterSpacing=0
     monospacedWidth=1
     size=24
     skewAmount=0
     typefaceFontName=Black
     outlineSettings {
      bApplyOutlineToDropShadows=false
      bMiteredCorners=false
      bSeparateFillAlpha=false
      outlineMaterial=null
      outlineSize=2
      outlineColor=<rgba 0.01 0.01 0.01 1>
     }
    }
    shadowOffset=<0 0>
    strikeBrush {
     bIsDynamicallyLoaded=false
     drawAs=NoDrawType
     imageType=NoImage
     mirroring=NoMirror
     resourceName=None
     resourceObject=null
     tiling=NoTile
     imageSize=<32 32>
     margin=<ltrb 0 0 0 0>
     outlineSettings {
      bUseBrushTransparency=false
      roundingType=HalfHeightRadius
      width=0
      color {colorUseRule=UseColor_Specified, specifiedColor=<rgba 0 0 0 0>}
      cornerRadii=<0 0 0 0>
     }
     tintColor {colorUseRule=UseColor_Specified, specifiedColor=<rgba 1 1 1 1>}
     uVRegion {bIsValid=false, max=<0 0>, min=<0 0>}
    }
   }
   slot {
    class=HorizontalBoxSlot
    properties {horizontalAlignment=HAlign_Center, verticalAlignment=VAlign_Center, padding=<ltrb 4 4 10 4>, size {sizeRule=Fill, value=1}}
   }
  }
  LODInfo {
   widgetClassPath=&MultiLineEditableTextBox
   bIsVariable=true
   parent=HorizontalBox_231
   widget=LODInfo
   properties {
    justification=Right
    text=Nanite: 10000 vtx
    widgetStyle {
     backgroundColor {colorUseRule=UseColor_Specified, specifiedColor=<rgba 1 1 1 1>}
     backgroundImageFocused {
      bIsDynamicallyLoaded=false
      drawAs=Box
      imageType=FullColor
      mirroring=NoMirror
      resourceName=../../../Engine/Content/Slate/Common/TextBox_Hovered.png
      resourceObject=null
      tiling=NoTile
      imageSize=<0 0>
      margin=<ltrb 0.25 0.25 0.25 0.25>
      outlineSettings {
       bUseBrushTransparency=false
       roundingType=HalfHeightRadius
       width=0
       color {colorUseRule=UseColor_Specified, specifiedColor=<rgba 0 0 0 0>}
       cornerRadii=<0 0 0 0>
      }
      tintColor {colorUseRule=UseColor_Specified, specifiedColor=<rgba 1 1 1 0>}
      uVRegion {bIsValid=false, max=<0 0>, min=<0 0>}
     }
     backgroundImageHovered {
      bIsDynamicallyLoaded=false
      drawAs=Box
      imageType=FullColor
      mirroring=NoMirror
      resourceName=../../../Engine/Content/Slate/Common/TextBox_Hovered.png
      resourceObject=null
      tiling=NoTile
      imageSize=<0 0>
      margin=<ltrb 0.25 0.25 0.25 0.25>
      outlineSettings {
       bUseBrushTransparency=false
       roundingType=HalfHeightRadius
       width=0
       color {colorUseRule=UseColor_Specified, specifiedColor=<rgba 0 0 0 0>}
       cornerRadii=<0 0 0 0>
      }
      tintColor {colorUseRule=UseColor_Specified, specifiedColor=<rgba 1 1 1 0>}
      uVRegion {bIsValid=false, max=<0 0>, min=<0 0>}
     }
     backgroundImageNormal {
      bIsDynamicallyLoaded=false
      drawAs=Box
      imageType=FullColor
      mirroring=NoMirror
      resourceName=../../../Engine/Content/Slate/Common/TextBox.png
      resourceObject=null
      tiling=NoTile
      imageSize=<0 0>
      margin=<ltrb 0.25 0.25 0.25 0.25>
      outlineSettings {
       bUseBrushTransparency=false
       roundingType=HalfHeightRadius
       width=0
       color {colorUseRule=UseColor_Specified, specifiedColor=<rgba 0 0 0 0>}
       cornerRadii=<0 0 0 0>
      }
      tintColor {colorUseRule=UseColor_Specified, specifiedColor=<rgba 1 1 1 0>}
      uVRegion {bIsValid=false, max=<0 0>, min=<0 0>}
     }
     backgroundImageReadOnly {
      bIsDynamicallyLoaded=false
      drawAs=Box
      imageType=FullColor
      mirroring=NoMirror
      resourceName=../../../Engine/Content/Slate/Common/TextBox_ReadOnly.png
      resourceObject=null
      tiling=NoTile
      imageSize=<0 0>
      margin=<ltrb 0.25 0.25 0.25 0.25>
      outlineSettings {
       bUseBrushTransparency=false
       roundingType=HalfHeightRadius
       width=0
       color {colorUseRule=UseColor_Specified, specifiedColor=<rgba 0 0 0 0>}
       cornerRadii=<0 0 0 0>
      }
      tintColor {colorUseRule=UseColor_Specified, specifiedColor=<rgba 1 1 1 1>}
      uVRegion {bIsValid=false, max=<0 0>, min=<0 0>}
     }
     focusedForegroundColor {colorUseRule=UseColor_Foreground, specifiedColor=<rgba 1 0 1 1>}
     foregroundColor {colorUseRule=UseColor_Foreground, specifiedColor=<rgba 1 0 1 1>}
     hScrollBarPadding=<ltrb 0 0 0 0>
     padding=<ltrb 4 2 4 2>
     readOnlyForegroundColor {colorUseRule=UseColor_Foreground, specifiedColor=<rgba 1 0 1 1>}
     scrollBarStyle {
      thickness=16
      draggedThumbImage {
       bIsDynamicallyLoaded=false
       drawAs=Box
       imageType=FullColor
       mirroring=NoMirror
       resourceName=../../../Engine/Content/Slate/Common/Scrollbar_Thumb.png
       resourceObject=null
       tiling=NoTile
       imageSize=<0 0>
       margin=<ltrb 0.25 0.25 0.25 0.25>
       outlineSettings {
        bUseBrushTransparency=false
        roundingType=HalfHeightRadius
        width=0
        color {colorUseRule=UseColor_Specified, specifiedColor=<rgba 0 0 0 0>}
        cornerRadii=<0 0 0 0>
       }
       tintColor {colorUseRule=UseColor_Specified, specifiedColor=<rgba 1 1 1 1>}
       uVRegion {bIsValid=false, max=<0 0>, min=<0 0>}
      }
      horizontalBackgroundImage {
       bIsDynamicallyLoaded=false
       drawAs=NoDrawType
       imageType=NoImage
       mirroring=NoMirror
       resourceName=None
       resourceObject=null
       tiling=NoTile
       imageSize=<0 0>
       margin=<ltrb 0 0 0 0>
       outlineSettings {
        bUseBrushTransparency=false
        roundingType=HalfHeightRadius
        width=0
        color {colorUseRule=UseColor_Specified, specifiedColor=<rgba 0 0 0 0>}
        cornerRadii=<0 0 0 0>
       }
       tintColor {colorUseRule=UseColor_Specified, specifiedColor=<rgba 1 1 1 1>}
       uVRegion {bIsValid=false, max=<0 0>, min=<0 0>}
      }
      horizontalBottomSlotImage {
       bIsDynamicallyLoaded=false
       drawAs=Image
       imageType=FullColor
       mirroring=NoMirror
       resourceName=../../../Engine/Content/Slate/Common/Scrollbar_Background_Horizontal.png
       resourceObject=null
       tiling=NoTile
       imageSize=<8 8>
       margin=<ltrb 0 0 0 0>
       outlineSettings {
        bUseBrushTransparency=false
        roundingType=HalfHeightRadius
        width=0
        color {colorUseRule=UseColor_Specified, specifiedColor=<rgba 0 0 0 0>}
        cornerRadii=<0 0 0 0>
       }
       tintColor {colorUseRule=UseColor_Specified, specifiedColor=<rgba 1 1 1 1>}
       uVRegion {bIsValid=false, max=<0 0>, min=<0 0>}
      }
      horizontalTopSlotImage {
       bIsDynamicallyLoaded=false
       drawAs=Image
       imageType=FullColor
       mirroring=NoMirror
       resourceName=../../../Engine/Content/Slate/Common/Scrollbar_Background_Horizontal.png
       resourceObject=null
       tiling=NoTile
       imageSize=<8 8>
       margin=<ltrb 0 0 0 0>
       outlineSettings {
        bUseBrushTransparency=false
        roundingType=HalfHeightRadius
        width=0
        color {colorUseRule=UseColor_Specified, specifiedColor=<rgba 0 0 0 0>}
        cornerRadii=<0 0 0 0>
       }
       tintColor {colorUseRule=UseColor_Specified, specifiedColor=<rgba 1 1 1 1>}
       uVRegion {bIsValid=false, max=<0 0>, min=<0 0>}
      }
      hoveredThumbImage {
       bIsDynamicallyLoaded=false
       drawAs=Box
       imageType=FullColor
       mirroring=NoMirror
       resourceName=../../../Engine/Content/Slate/Common/Scrollbar_Thumb.png
       resourceObject=null
       tiling=NoTile
       imageSize=<0 0>
       margin=<ltrb 0.25 0.25 0.25 0.25>
       outlineSettings {
        bUseBrushTransparency=false
        roundingType=HalfHeightRadius
        width=0
        color {colorUseRule=UseColor_Specified, specifiedColor=<rgba 0 0 0 0>}
        cornerRadii=<0 0 0 0>
       }
       tintColor {colorUseRule=UseColor_Specified, specifiedColor=<rgba 1 1 1 1>}
       uVRegion {bIsValid=false, max=<0 0>, min=<0 0>}
      }
      normalThumbImage {
       bIsDynamicallyLoaded=false
       drawAs=Box
       imageType=FullColor
       mirroring=NoMirror
       resourceName=../../../Engine/Content/Slate/Common/Scrollbar_Thumb.png
       resourceObject=null
       tiling=NoTile
       imageSize=<0 0>
       margin=<ltrb 0.25 0.25 0.25 0.25>
       outlineSettings {
        bUseBrushTransparency=false
        roundingType=HalfHeightRadius
        width=0
        color {colorUseRule=UseColor_Specified, specifiedColor=<rgba 0 0 0 0>}
        cornerRadii=<0 0 0 0>
       }
       tintColor {colorUseRule=UseColor_Specified, specifiedColor=<rgba 1 1 1 1>}
       uVRegion {bIsValid=false, max=<0 0>, min=<0 0>}
      }
      verticalBackgroundImage {
       bIsDynamicallyLoaded=false
       drawAs=NoDrawType
       imageType=NoImage
       mirroring=NoMirror
       resourceName=None
       resourceObject=null
       tiling=NoTile
       imageSize=<0 0>
       margin=<ltrb 0 0 0 0>
       outlineSettings {
        bUseBrushTransparency=false
        roundingType=HalfHeightRadius
        width=0
        color {colorUseRule=UseColor_Specified, specifiedColor=<rgba 0 0 0 0>}
        cornerRadii=<0 0 0 0>
       }
       tintColor {colorUseRule=UseColor_Specified, specifiedColor=<rgba 1 1 1 1>}
       uVRegion {bIsValid=false, max=<0 0>, min=<0 0>}
      }
      verticalBottomSlotImage {
       bIsDynamicallyLoaded=false
       drawAs=Image
       imageType=FullColor
       mirroring=NoMirror
       resourceName=../../../Engine/Content/Slate/Common/Scrollbar_Background_Vertical.png
       resourceObject=null
       tiling=NoTile
       imageSize=<8 8>
       margin=<ltrb 0 0 0 0>
       outlineSettings {
        bUseBrushTransparency=false
        roundingType=HalfHeightRadius
        width=0
        color {colorUseRule=UseColor_Specified, specifiedColor=<rgba 0 0 0 0>}
        cornerRadii=<0 0 0 0>
       }
       tintColor {colorUseRule=UseColor_Specified, specifiedColor=<rgba 1 1 1 1>}
       uVRegion {bIsValid=false, max=<0 0>, min=<0 0>}
      }
      verticalTopSlotImage {
       bIsDynamicallyLoaded=false
       drawAs=Image
       imageType=FullColor
       mirroring=NoMirror
       resourceName=../../../Engine/Content/Slate/Common/Scrollbar_Background_Vertical.png
       resourceObject=null
       tiling=NoTile
       imageSize=<8 8>
       margin=<ltrb 0 0 0 0>
       outlineSettings {
        bUseBrushTransparency=false
        roundingType=HalfHeightRadius
        width=0
        color {colorUseRule=UseColor_Specified, specifiedColor=<rgba 0 0 0 0>}
        cornerRadii=<0 0 0 0>
       }
       tintColor {colorUseRule=UseColor_Specified, specifiedColor=<rgba 1 1 1 1>}
       uVRegion {bIsValid=false, max=<0 0>, min=<0 0>}
      }
     }
     textStyle {
      overflowPolicy=Clip
      transformPolicy=None
      colorAndOpacity {colorUseRule=UseColor_Specified, specifiedColor=<rgba 1 1 1 1>}
      font {
       bForceMonospaced=false
       bMaterialIsStencil=false
       fontMaterial=null
       fontObject=&DroidSansMono
       letterSpacing=0
       monospacedWidth=1
       size=18
       skewAmount=0
       typefaceFontName=Default
       outlineSettings {
        bApplyOutlineToDropShadows=false
        bMiteredCorners=false
        bSeparateFillAlpha=false
        outlineMaterial=null
        outlineSize=1
        outlineColor=<rgba 0 0 0 1>
       }
      }
      highlightColor {colorUseRule=UseColor_Specified, specifiedColor=<rgba 0.02 0.3 0 1>}
      highlightShape {
       bIsDynamicallyLoaded=false
       drawAs=Box
       imageType=FullColor
       mirroring=NoMirror
       resourceName=../../../Engine/Content/Slate/Common/TextBlockHighlightShape.png
       resourceObject=null
       tiling=NoTile
       imageSize=<0 0>
       margin=<ltrb 0.375 0.375 0.375 0.375>
       outlineSettings {
        bUseBrushTransparency=false
        roundingType=HalfHeightRadius
        width=0
        color {colorUseRule=UseColor_Specified, specifiedColor=<rgba 0 0 0 0>}
        cornerRadii=<0 0 0 0>
       }
       tintColor {colorUseRule=UseColor_Specified, specifiedColor=<rgba 1 1 1 1>}
       uVRegion {bIsValid=false, max=<0 0>, min=<0 0>}
      }
      selectedBackgroundColor {colorUseRule=UseColor_Foreground, specifiedColor=<rgba 1 0 1 1>}
      shadowColorAndOpacity=<rgba 0 0 0 1>
      shadowOffset=<0 0>
      strikeBrush {
       bIsDynamicallyLoaded=false
       drawAs=Image
       imageType=NoImage
       mirroring=NoMirror
       resourceName=None
       resourceObject=null
       tiling=NoTile
       imageSize=<32 32>
       margin=<ltrb 0 0 0 0>
       outlineSettings {
        bUseBrushTransparency=false
        roundingType=HalfHeightRadius
        width=0
        color {colorUseRule=UseColor_Specified, specifiedColor=<rgba 0 0 0 0>}
        cornerRadii=<0 0 0 0>
       }
       tintColor {colorUseRule=UseColor_Specified, specifiedColor=<rgba 1 1 1 1>}
       uVRegion {bIsValid=false, max=<0 0>, min=<0 0>}
      }
      underlineBrush {
       bIsDynamicallyLoaded=false
       drawAs=Image
       imageType=NoImage
       mirroring=NoMirror
       resourceName=None
       resourceObject=null
       tiling=NoTile
       imageSize=<32 32>
       margin=<ltrb 0 0 0 0>
       outlineSettings {
        bUseBrushTransparency=false
        roundingType=HalfHeightRadius
        width=0
        color {colorUseRule=UseColor_Specified, specifiedColor=<rgba 0 0 0 0>}
        cornerRadii=<0 0 0 0>
       }
       tintColor {colorUseRule=UseColor_Specified, specifiedColor=<rgba 1 1 1 1>}
       uVRegion {bIsValid=false, max=<0 0>, min=<0 0>}
      }
     }
     vScrollBarPadding=<ltrb 0 0 0 0>
    }
   }
   slot {class=HorizontalBoxSlot, properties {verticalAlignment=VAlign_Center, size {sizeRule=Fill, value=1}}}
  }
  Border {
   widgetClassPath=&Border
   parent=VerticalBox_41
   widget=Border
   properties {
    background {
     bIsDynamicallyLoaded=false
     drawAs=Image
     imageType=NoImage
     mirroring=NoMirror
     resourceName=None
     resourceObject=null
     tiling=NoTile
     imageSize=<32 32>
     margin=<ltrb 0 0 0 0>
     outlineSettings {
      bUseBrushTransparency=false
      roundingType=HalfHeightRadius
      width=0
      color {colorUseRule=UseColor_Specified, specifiedColor=<rgba 0 0 0 0>}
      cornerRadii=<0 0 0 0>
     }
     tintColor {colorUseRule=UseColor_Specified, specifiedColor=<rgba 0 0 0 0.5>}
     uVRegion {bIsValid=false, max=<0 0>, min=<0 0>}
    }
    padding=<ltrb 4 4 4 48>
   }
   slot {class=VerticalBoxSlot, properties {verticalAlignment=VAlign_Top}}
  }
  HorizontalBox {
   widgetClassPath=&HorizontalBox
   parent=Border
   widget=HorizontalBox
   slot {class=BorderSlot, properties {padding=<ltrb 4 4 4 48>}}
  }
  BottomLeft {
   widgetClassPath=&TextBlock
   bIsVariable=true
   parent=HorizontalBox
   widget=BottomLeft
   properties {
    text="Master:   [00:00:14:8] (344f)"
    font {
     bForceMonospaced=false
     bMaterialIsStencil=false
     fontMaterial=null
     fontObject=&NotoSans
     letterSpacing=0
     monospacedWidth=1
     size=18
     skewAmount=0
     typefaceFontName=NotoSans-Regular
     outlineSettings {
      bApplyOutlineToDropShadows=false
      bMiteredCorners=false
      bSeparateFillAlpha=false
      outlineMaterial=null
      outlineSize=0
      outlineColor=<rgba 0 0 0 1>
     }
    }
   }
   slot {class=HorizontalBoxSlot, properties {horizontalAlignment=HAlign_Left, size {sizeRule=Automatic, value=0.5}}}
  }
  Spacer_256 {
   widgetClassPath=&Spacer
   parent=HorizontalBox
   widget=Spacer_256
   slot {class=HorizontalBoxSlot, properties {size {sizeRule=Fill, value=1}}}
  }
  BottomLeft_tod {
   widgetClassPath=&TextBlock
   bIsVariable=true
   parent=HorizontalBox
   widget=BottomLeft_tod
   properties {
    text=Time Of Day:
    font {
     bForceMonospaced=false
     bMaterialIsStencil=false
     fontMaterial=null
     fontObject=&NotoSans
     letterSpacing=0
     monospacedWidth=1
     size=18
     skewAmount=0
     typefaceFontName=NotoSans-Regular
     outlineSettings {
      bApplyOutlineToDropShadows=false
      bMiteredCorners=false
      bSeparateFillAlpha=false
      outlineMaterial=null
      outlineSize=0
      outlineColor=<rgba 0 0 0 1>
     }
    }
   }
   slot {class=HorizontalBoxSlot, properties {horizontalAlignment=HAlign_Left, size {sizeRule=Automatic, value=0.5}}}
  }
  Spacer_3 {
   widgetClassPath=&Spacer
   parent=HorizontalBox
   widget=Spacer_3
   slot {class=HorizontalBoxSlot, properties {size {sizeRule=Fill, value=1}}}
  }
  BottomLeft2 {
   widgetClassPath=&TextBlock
   bIsVariable=true
   parent=HorizontalBox
   widget=BottomLeft2
   properties {
    text="TSS: "
    font {
     bForceMonospaced=false
     bMaterialIsStencil=false
     fontMaterial=null
     fontObject=&NotoSans
     letterSpacing=0
     monospacedWidth=1
     size=18
     skewAmount=0
     typefaceFontName=NotoSans-Regular
     outlineSettings {
      bApplyOutlineToDropShadows=false
      bMiteredCorners=false
      bSeparateFillAlpha=false
      outlineMaterial=null
      outlineSize=0
      outlineColor=<rgba 0 0 0 1>
     }
    }
   }
   slot {class=HorizontalBoxSlot, properties {horizontalAlignment=HAlign_Center}}
  }
  Spacer_399 {
   widgetClassPath=&Spacer
   parent=HorizontalBox
   widget=Spacer_399
   slot {class=HorizontalBoxSlot, properties {size {sizeRule=Fill, value=1}}}
  }
  BottomLeft_FOV {
   widgetClassPath=&TextBlock
   bIsVariable=true
   parent=HorizontalBox
   widget=BottomLeft_FOV
   properties {
    text=FOV
    font {
     bForceMonospaced=false
     bMaterialIsStencil=false
     fontMaterial=null
     fontObject=&NotoSans
     letterSpacing=0
     monospacedWidth=1
     size=18
     skewAmount=0
     typefaceFontName=NotoSans-Regular
     outlineSettings {
      bApplyOutlineToDropShadows=false
      bMiteredCorners=false
      bSeparateFillAlpha=false
      outlineMaterial=null
      outlineSize=0
      outlineColor=<rgba 0 0 0 1>
     }
    }
   }
   slot {class=HorizontalBoxSlot, properties {horizontalAlignment=HAlign_Center}}
  }
  Spacer_4 {
   widgetClassPath=&Spacer
   parent=HorizontalBox
   widget=Spacer_4
   slot {class=HorizontalBoxSlot, properties {size {sizeRule=Fill, value=1}}}
  }
  BottomCenter {
   widgetClassPath=&TextBlock
   bIsVariable=true
   parent=HorizontalBox
   widget=BottomCenter
   properties {
    text="60mm   f/2.5"
    font {
     bForceMonospaced=false
     bMaterialIsStencil=false
     fontMaterial=null
     fontObject=&NotoSans
     letterSpacing=0
     monospacedWidth=1
     size=18
     skewAmount=0
     typefaceFontName=NotoSans-Regular
     outlineSettings {
      bApplyOutlineToDropShadows=false
      bMiteredCorners=false
      bSeparateFillAlpha=false
      outlineMaterial=null
      outlineSize=0
      outlineColor=<rgba 0 0 0 1>
     }
    }
   }
   slot {class=HorizontalBoxSlot, properties {horizontalAlignment=HAlign_Center}}
  }
  Spacer_554 {
   widgetClassPath=&Spacer
   parent=HorizontalBox
   widget=Spacer_554
   slot {class=HorizontalBoxSlot, properties {size {sizeRule=Fill, value=1}}}
  }
  HorizontalBox_0 {
   widgetClassPath=&HorizontalBox
   parent=HorizontalBox
   widget=HorizontalBox_0
   slot {class=HorizontalBoxSlot, properties {horizontalAlignment=HAlign_Right, size {sizeRule=Automatic, value=0.5}}}
  }
  BottomRight {
   widgetClassPath=&TextBlock
   bIsVariable=true
   parent=HorizontalBox_0
   widget=BottomRight
   properties {
    justification=Right
    text="Shot:   Cam: CineCameraActor1 [00:00:3:13] (109f)"
    font {
     bForceMonospaced=false
     bMaterialIsStencil=false
     fontMaterial=null
     fontObject=&NotoSans
     letterSpacing=0
     monospacedWidth=1
     size=18
     skewAmount=0
     typefaceFontName=NotoSans-Regular
     outlineSettings {
      bApplyOutlineToDropShadows=false
      bMiteredCorners=false
      bSeparateFillAlpha=false
      outlineMaterial=null
      outlineSize=0
      outlineColor=<rgba 0 0 0 1>
     }
    }
   }
   slot {class=HorizontalBoxSlot}
  }
  BottomRight2 {
   widgetClassPath=&TextBlock
   bIsVariable=true
   parent=HorizontalBox_0
   widget=BottomRight2
   properties {
    justification=Right
    text=RenderFarmSeqB
    font {
     bForceMonospaced=false
     bMaterialIsStencil=false
     fontMaterial=null
     fontObject=&NotoSans
     letterSpacing=0
     monospacedWidth=1
     size=18
     skewAmount=0
     typefaceFontName=NotoSans-Bold
     outlineSettings {
      bApplyOutlineToDropShadows=false
      bMiteredCorners=false
      bSeparateFillAlpha=false
      outlineMaterial=null
      outlineSize=0
      outlineColor=<rgba 0 0 0 1>
     }
    }
   }
   slot {class=HorizontalBoxSlot, properties {padding=<ltrb 50 0 0 0>}}
  }
 ]
}
